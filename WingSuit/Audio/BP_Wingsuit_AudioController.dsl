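; Wingsuit audio controller component that spawns attached first- or third-person flight audio on the owning pawn, and destroys itself where cosmetic events cannot run.
; For locally viewed pawns, a looping timer sends anim-instance roll to the sound and fires an async downward trace.
; Trace results update the close-to-ground flag and hit distance as audio parameters.
; StopAudio fades out the sound and destroys the component.
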(event EventBeginPlay
  ;[1 1 1] Trigger Audio Playback
  (Networking|CanExecuteCosmeticEvents
    (:True
      (bind _asfort_player_pawn (Utilities|Casting|CastToFortPlayerPawn (Components|GetOwner)))
      (bind _output_get (Variables|Default|SetPlayerPawn _asfort_player_pawn))
      (bind _returnvalue (Audio|SpawnSoundAttached (select (Pawn|IsLocallyViewed _output_get) (Variables|Default|Get1pAudio) (Variables|Default|Get3pAudio)) (Variables|/Script/Engine.Actor|GetRootComponent _output_get)))
      (Variables|Default|SetFlyingAudio _returnvalue)
      (if (Pawn|IsLocallyViewed _output_get)
        (CallFunction|MovementUpdate)))
    (:False
      (Components|DestroyComponent))))

(event Custom|MovementUpdate
  ;[1 1 1] Timer event setup
  (Audio|Parameter|SetFloatParameter (Variables|Default|GetFlyingAudio) "MaxCloseDist" (Variables|Default|GetMaxCloseDist) :!class /Script/AudioExtensions.AudioParameterControllerInterface)
  (Utilities|Time|SetTimerbyEvent (EventDispatchers|CreateEvent :!event Custom|LocationCheck) (Variables|Default|GetMovementInterval) true))

(event Custom|StopAudio
  ;[1 1 1] Stop audio and destroy component
  (bind _flyingaudio (Variables|Default|GetFlyingAudio)
    (:then
      (Audio|Components|Audio|FadeOut _flyingaudio 0.5 0.0 "Linear" :!class /Script/Engine.AudioComponent))
    (:else
      _))
  (Components|DestroyComponent))

(event Custom|LocationCheck
  ;[1 1 1] Decide if the player has made a directional movement, and if they are close to the ground
  (bind _animationbp (Variables|Default|GetAnimationBp)
    (:then
      (Audio|Parameter|SetFloatParameter (Variables|Default|GetFlyingAudio) "Roll" (Variables|/Script/WingSuitRuntime.WingsuitGadgetCCPAnimInstance|GetRoll _animationbp) :!class /Script/AudioExtensions.AudioParameterControllerInterface)
      (bind _output_get (Variables|Default|SetPlayer%20Location (Transformation|GetActorLocation (Variables|Default|GetPlayerPawn))))
      (Class|AsyncTraceQueue|AsyncTraceImmediate _output_get (Math|Vector|MakeVector (.x _output_get) (.y _output_get) (- (.z _output_get) (Variables|Default|GetMaxCloseDist))) (Utilities|Array|MakeArray (Components|GetOwner)) "ECC_Visibility" (delegate Custom|Async%20Trace%20Completed)))
    (:else)))

(event Custom|Async%20Trace%20Completed
  (params (Succeeded :type Boolean) (ActorHit :type Actor%20Object%20Reference) (HitResult :type Hit%20Result%20Structure))
  (bind _hit_result (Variables|Default|GetHit%20Result))
  (bind (_bblockinghit _binitialoverlap _time _distance) (Collision|BreakHitResult _hit_result))
  ;[1 1 1] Decide if the player has made a directional movement, and if they are close to the ground
  (Variables|Default|SetHit%20Result HitResult)
  (if (!= (Variables|Default|GetbIsCloseToGround) _bblockinghit)
    (bind _output_get (Variables|Default|SetbIsCloseToGround (not (Variables|Default|GetbIsCloseToGround))))
    (Audio|Parameter|SetBooleanParameter (Variables|Default|GetFlyingAudio) "bIsCloseToGround" _output_get :!class /Script/AudioExtensions.AudioParameterControllerInterface)
    (else
      _))
  (Variables|Default|GetbIsCloseToGround
    (:then
      (Audio|Parameter|SetFloatParameter (Variables|Default|GetFlyingAudio) "Distance" _distance :!class /Script/AudioExtensions.AudioParameterControllerInterface))
    (:else)))
