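# Mono MetaSound source for the looping wingsuit cooldown sound, routed to the gadget sound class with foley attenuation.
# On Play starts an ADSR envelope, whose On Attack Triggered output starts a wave player with the cooldown loop asset.
# The envelope multiplies the audio, so it swells in, then slowly decays to a quiet sustain with no release trigger wired.
# The source overrides concurrency to cap simultaneous instances.
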MetaSoundSource /WingSuit/Audio/MSS_WingsuitCoolDown_Loop
inputs [UE.Source.OnPlay {type=Trigger}]
interfaces [UE.OutputFormat.Mono:1.0, UE.Source:1.1]
outputs ["UE.OutputFormat.Mono.Audio:0" {type=Audio}]
pages [Default {
  input_defaults {UE.Source.OnPlay=null}
  nodes [
   ADSREnvelopeADSREnvelope_0 = ADSR Envelope.ADSR Envelope:Float(Trigger Attack=^UE.Source.OnPlay) {also {defaults {Attack Time=1, Decay Time=8, Sustain Level=0.2}}}
   WavePlayer_0 = UE.Wave Player:Mono(Play=@ADSREnvelopeADSREnvelope_0.On Attack Triggered) {also {defaults {Loop="<MetaSoundBuilderResult.SUCCEEDED: 0>", Wave Asset=/WingSuit/Audio/wav/Wingsuit_Cooldown_Loop_01}, version [1,1]}}
   Multiply_0 = UE.Multiply:Audio by Float(AdditionalOperands=UE.Reroute(Value=@ADSREnvelopeADSREnvelope_0.Out Envelope).Value, PrimaryOperand=@WavePlayer_0.Out Mono)
  ]
  links [{dst="@UE.OutputFormat.Mono.Audio:0", src=Multiply_0.Out}]
 }]
properties {
 assetImportData=&self:AssetImportData
 attenuationSettings=&/Att_WingSuit_Foley
 bLooping=true
 bOverrideConcurrency=true
 duration=10000
 maxDistance=1100
 sampleRate=48000
 soundClassObject=&Fort_Gadgets_Items
 volume=0.636
 concurrencyOverrides {maxCount=3, retriggerTime=1}
}
