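# MetaSound patch for spatial travel audio that turns elevation movement into a normalized vertical value and Rise, Fall and Stop triggers.
# It maps Elevation.Movement between floor and roof to -1..1, smooths it, and tests each change against a symmetric IgnoreRange dead zone.
# Trigger Control gates wired through deferred variables close after firing: Stop rearms Rise and Fall, and either of those rearms Stop, preventing repeats.
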MetaSoundPatch /FortMetasound/SpatialTravel/MSP_SpatialTravelVerticalDetection
inputs [
 Elevation.Movement {type=Float}
 IgnoreRange {type=Float}
 VerticalFloor {type=Float}
 VerticalRoof {type=Float}
]
outputs [
 Fall {type=Trigger}
 NormalizedVertical {type=Float}
 Rise {type=Trigger}
 Stop {type=Trigger}
]
pages [Default {
  input_defaults {Elevation.Movement=0, IgnoreRange=0.1, VerticalFloor=-90, VerticalRoof=90}
  nodes [
   InitVariable.Float()
   InitVariable.Float()
   InitVariable.Float()
   InterpTo_0 = UE.InterpTo:Audio(Target=MapRange.MapRange:Float(In=^Elevation.Movement, In Range A=^VerticalFloor, In Range B=^VerticalRoof) {also {defaults {Out Range A=-1}}}.Out Value) {also {defaults {Interp Time=0.25}}}
   n1 = UE.Reroute(Value=@InterpTo_0.Value)
   n2 = UE.Reroute(Value=@n1.Value)
   n3 = VariableAccessor.Float(Variable=VariableMutator.Float(Variable=InitVariable.Float().Variable, Value=UE.Multiply:Float(PrimaryOperand=^IgnoreRange) {also {defaults {AdditionalOperands=-1}}}.Out).Variable)
   TriggerOnValueChange_0 = UE.Trigger On Value Change:Float(Value=@InterpTo_0.Value)
   n4 = TriggerCompare.Clamp:Bool(A=/FortMetasound/Patches/InRangeFloat(Max=^IgnoreRange, Input=@n1.Value, Min=@n3.Value).Result, Compare=@TriggerOnValueChange_0.Trigger) {also {defaults {B="<MetaSoundBuilderResult.SUCCEEDED: 0>"}}}
   n5 = VariableDeferredAccessor.Trigger(Variable=InitVariable.Trigger().Variable)
   n6 = VariableDeferredAccessor.Trigger(Variable=InitVariable.Trigger().Variable)
   TriggerAnyTriggerAny2_0 = "TriggerAny.Trigger Any (2)"(In 0=@n5.Value, In 1=@n6.Value)
   n7 = VariableDeferredAccessor.Trigger(Variable=InitVariable.Trigger().Variable)
   TriggerControl_1 = UE.Trigger Control(Open=@TriggerAnyTriggerAny2_0.Out, Trigger In=@n4.True, Close=@n7.Value)
   n8 = VariableAccessor.Trigger(Variable=VariableMutator.Trigger(Value=@TriggerControl_1.Trigger Out, Variable=@n7.Variable).Variable)
   n9 = VariableDeferredAccessor.Trigger(Variable=@n6.Variable)
   TriggerControl_2 = UE.Trigger Control(
    Trigger In=TriggerCompare.Clamp:Float(B=^IgnoreRange, A=@n2.Value, Compare=@n4.False) {also {defaults {Type=3}}}.True
    Open=@n8.Value
    Close=@n9.Value
   )
   VariableMutator.Trigger(Value=@TriggerControl_2.Trigger Out, Variable=@n9.Variable)
   n10 = VariableDeferredAccessor.Trigger(Variable=@n5.Variable)
   TriggerControl_0 = UE.Trigger Control(
    Trigger In=TriggerCompare.Clamp:Float(A=@n2.Value, Compare=@n4.False, B=VariableAccessor.Float(Variable=@n3.Variable).Value) {also {defaults {Type=2}}}.True
    Open=@n8.Value
    Close=@n10.Value
   )
   VariableMutator.Trigger(Value=@TriggerControl_0.Trigger Out, Variable=@n10.Variable)
  ]
  links [
   {dst="@NormalizedVertical", src=InterpTo_0.Value}
   {dst="@Fall", src=TriggerControl_0.Trigger Out}
   {dst="@Stop", src=TriggerControl_1.Trigger Out}
   {dst="@Rise", src=TriggerControl_2.Trigger Out}
  ]
 }]
