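# Looping stereo MetaSound source for first-person wingsuit movement audio.
# Interpolated speed scales the glide loop, vertical-movement loop and ground-proximity gain; the proximity loop fades on a close-to-ground bool, with distance shaped by a wavetable curve.
# Roll gates in each direction fade turn loops, whose starts trigger random wing one-shots, before loop, one-shot and master gains.
# The Mobile page drops vertical and turn loops, gating fewer one-shots on absolute roll.
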MetaSoundSource /WingSuit/Audio/MSS_Wingsuit_Movement_1p
inputs [
 bIsCloseToGround {type=Bool}
 Distance {type=Float}
 GroundCloseLoop {type=WaveAsset:Array}
 Loop.MainLoopVolMax {type=Float}
 Loop.MainLoopVolMin {type=Float}
 MainLoop {type=WaveAsset:Array}
 MaxCloseDist {type=Float}
 Mix.LoopGain {type=Float}
 Mix.MasterGain {type=Float}
 Mix.OneShotGain {type=Float}
 Movement.Speed {type=Float}
 Movement.SpeedAzimuth {type=Float}
 Movement.SpeedElevation {type=Float}
 OneShots.Turn {type=WaveAsset:Array}
 Roll {type=Float}
 UE.Source.OnPlay {type=Trigger}
]
interfaces [Fort.Movement:1.0, UE.OutputFormat.Stereo:1.0, UE.Source:1.1]
outputs [
 "UE.OutputFormat.Stereo.Audio:0" {type=Audio}
 "UE.OutputFormat.Stereo.Audio:1" {type=Audio}
]
pages [
 Default {
  is_build_graph=true
  input_defaults {
   Distance=1
   Loop.MainLoopVolMax=1.3
   Loop.MainLoopVolMin=0.6
   MaxCloseDist=1500
   Mix.LoopGain=1
   Mix.MasterGain=1
   Mix.OneShotGain=1
   Movement.Speed=0
   Movement.SpeedAzimuth=0
   Movement.SpeedElevation=0
   Roll=1
   UE.Source.OnPlay=null
   bIsCloseToGround=null
   GroundCloseLoop [/WingSuit/Audio/wav/Wingsuit_Flying_Down_Loop_01]
   MainLoop [/WingSuit/Audio/wav/Wingsuit_Flying_Glide_Loop_01]
   OneShots.Turn [
    /WingSuit/Audio/wav/Wingsuit_Wings_Turn_Gentle_01
    /WingSuit/Audio/wav/Wingsuit_Wings_Turn_Gentle_02
    /WingSuit/Audio/wav/Wingsuit_Wings_Turn_Gentle_03
    /WingSuit/Audio/wav/Wingsuit_Wings_Turn_Gentle_04
    /WingSuit/Audio/wav/Wingsuit_Wings_Turn_Gentle_05
    /WingSuit/Audio/wav/Wingsuit_Wings_Turn_Gentle_06
    /WingSuit/Audio/wav/Wingsuit_Wings_Turn_Gentle_07
   ]
  }
  nodes [
   n1 = UE.InterpTo:Audio(Target=^Roll)
   n2 = &MSP_OneshotGate(Value=@n1.Value) {also {defaults {EnterCondition=3, ExitCondition=4, OneshotThreshold=0.7}}}
   n3 = /FortMetasound/SpatialTravel/MSP_SpatialTravelLoop(FadeIn=@n2.OnGateEnter, FadeOut=@n2.OnGateExit) {also {defaults {Loop.FadeTime=2, Loop.PitchControl=0, Loop.PitchhMin=0.7}}}
   n4 = &MSP_OneshotGate(Value=@n1.Value) {also {defaults {EnterCondition=2, ExitCondition=5, OneshotThreshold=-0.7}}}
   n5 = &MSP_SpatialTravelLoop(FadeIn=@n4.OnGateEnter, FadeOut=@n4.OnGateExit) {also {defaults {Loop.FadeTime=2, Loop.PitchControl=0, Loop.PitchhMin=0.7}}}
   TriggerOnValueChange_0 = UE.Trigger On Value Change:Bool(Value=^bIsCloseToGround)
   n6 = TriggerCompare.Clamp:Bool(A=^bIsCloseToGround, Compare=@TriggerOnValueChange_0.Trigger) {also {defaults {B="<MetaSoundBuilderResult.SUCCEEDED: 0>"}}}
   n7 = UE.InterpTo:Audio(Target=^Distance)
   n8 = &MSP_SpatialTravelLoop(
    Loop.Sounds=^GroundCloseLoop
    FadeOut=@n6.False
    FadeIn=@n6.True
    Loop.VolumeControl=UE.WaveTableBankEvaluate(Input=MapRange.MapRange:Float(In Range B=^MaxCloseDist, In=@n7.Value) {also {defaults {Out Range A=1.5, Out Range B=0.2}}}.Out Value) {also {defaults {WaveTableBank=&/WTB_Wingsuit_CloseToGroundCurve}}}.Output
   ) {also {defaults {Loop.FadeTime=0.1, Loop.PitchControl=0}}}
   n9 = MapRange.MapRange:Float(In=^Movement.Speed) {also {defaults {In Range A=700, In Range B=2200}}}
   n10 = VariableAccessor.Float(Variable=VariableMutator.Float(Variable=InitVariable.Float().Variable, Value=UE.InterpTo:Audio(Target=@n9.Out Value).Value).Variable)
   n11 = VariableAccessor.Float(Variable=@n10.Variable)
   n12 = InitVariable.Trigger()
   n13 = /FortMetasound/SpatialTravel/MSP_SpatialTravelVerticalDetection(Elevation.Movement=^Movement.SpeedElevation) {also {defaults {VerticalFloor=-60, VerticalRoof=60}}}
   n14 = InitVariable.Trigger()
   n15 = &MSP_SpatialTravelLoop(
    Loop.VolumeControl=MapRange.MapRange:Float(In=VariableAccessor.Float(Variable=@n11.Variable).Value) {also {defaults {Out Range A=0.5, Out Range B=1.3}}}.Out Value
    FadeIn=VariableAccessor.Trigger(Variable=VariableMutator.Trigger(Variable=@n12.Variable, Value=@n13.Rise).Variable).Value
    FadeOut=VariableAccessor.Trigger(Variable=VariableMutator.Trigger(Variable=@n14.Variable, Value=@n13.Fall).Variable).Value
   ) {also {defaults {Loop.FadeTime=1.5, Loop.PitchControl=0, Loop.PitchhMin=0.7}}}
   n16 = &MSP_SpatialTravelLoop(
    Loop.Sounds=^MainLoop
    FadeIn=^UE.Source.OnPlay
    Loop.VolumeControl=MapRange.MapRange:Float(Out Range B=^Loop.MainLoopVolMax, Out Range A=^Loop.MainLoopVolMin, In=@n10.Value).Out Value
   )
   AudioMixerAudioMixerStereo4_0 = "AudioMixer.Audio Mixer (Stereo, 4)"(
    In 3 R=UE.Add:Audio(AdditionalOperands=@n3.OutR, PrimaryOperand=@n5.OutR).Out
    In 3 L=UE.Add:Audio(PrimaryOperand=@n3.OutL, AdditionalOperands=@n5.OutL).Out
    In 2 L=@n8.OutL
    In 2 R=@n8.OutR
    In 1 L=@n15.OutL
    In 1 R=@n15.OutR
    In 0 L=@n16.OutL
    In 0 R=@n16.OutR
    Gain 2=MapRange.MapRange:Float(In=@n11.Value) {also {defaults {Out Range A=0.1, Out Range B=1.7}}}.Out Value
   ) {also {defaults {Gain 1=1.5, Gain 3=1.5}}}
   n17 = InitVariable.Trigger()
   TriggerAnyTriggerAny2_0 = "TriggerAny.Trigger Any (2)"(In 1=@n3.On Play, In 0=@n5.On Play)
   ArrayRandomGet_0 = Array.Random Get:WaveAsset:Array(
    In Array=^OneShots.Turn
    Next=VariableAccessor.Trigger(Variable=VariableMutator.Trigger(Variable=@n17.Variable, Value=@TriggerAnyTriggerAny2_0.Out).Variable).Value
   )
   WavePlayer_0 = UE.Wave Player:Stereo(Play=@ArrayRandomGet_0.On Next, Wave Asset=@ArrayRandomGet_0.Value) {also {version [1,1]}}
   AudioMixerAudioMixerStereo2_0 = "AudioMixer.Audio Mixer (Stereo, 2)"(
    Gain 0=^Mix.LoopGain
    Gain 1=^Mix.OneShotGain
    In 0 L=@AudioMixerAudioMixerStereo4_0.Out L
    In 0 R=@AudioMixerAudioMixerStereo4_0.Out R
    In 1 L=@WavePlayer_0.Out Left
    In 1 R=@WavePlayer_0.Out Right
   )
   Multiply_0 = UE.Multiply:Audio by Float(AdditionalOperands=^Mix.MasterGain, PrimaryOperand=@AudioMixerAudioMixerStereo2_0.Out R)
   Multiply_1 = UE.Multiply:Audio by Float(AdditionalOperands=^Mix.MasterGain, PrimaryOperand=@AudioMixerAudioMixerStereo2_0.Out L)
  ]
  links [{dst="@UE.OutputFormat.Stereo.Audio:1", src=Multiply_0.Out}, {dst="@UE.OutputFormat.Stereo.Audio:0", src=Multiply_1.Out}]
 }
 Mobile {
  input_defaults {OneShots.Turn [
    &Wingsuit_Wings_Turn_Gentle_01
    &Wingsuit_Wings_Turn_Gentle_05
    &Wingsuit_Wings_Turn_Gentle_07
   ]}
  nodes [
   TriggerOnValueChange_0 = UE.Trigger On Value Change:Bool(Value=^bIsCloseToGround)
   n1 = TriggerCompare.Clamp:Bool(A=^bIsCloseToGround, Compare=@TriggerOnValueChange_0.Trigger) {also {defaults {B="<MetaSoundBuilderResult.SUCCEEDED: 0>"}}}
   n2 = UE.InterpTo:Audio(Target=^Distance)
   n3 = &MSP_SpatialTravelLoop(
    Loop.Sounds=^GroundCloseLoop
    FadeOut=@n1.False
    FadeIn=@n1.True
    Loop.VolumeControl=UE.WaveTableBankEvaluate(Input=MapRange.MapRange:Float(In Range B=^MaxCloseDist, In=@n2.Value) {also {defaults {Out Range A=1, Out Range B=0}}}.Out Value) {also {defaults {WaveTableBank=&WTB_Wingsuit_CloseToGroundCurve}}}.Output
   ) {also {defaults {Loop.FadeTime=0.1, Loop.PitchControl=0}}}
   n4 = MapRange.MapRange:Float(In=^Movement.Speed) {also {defaults {In Range A=700, In Range B=2200}}}
   n5 = VariableAccessor.Float(Variable=VariableMutator.Float(Variable=InitVariable.Float().Variable, Value=UE.InterpTo:Audio(Target=@n4.Out Value).Value).Variable)
   n6 = &MSP_SpatialTravelLoop(
    Loop.Sounds=^MainLoop
    FadeIn=^UE.Source.OnPlay
    Loop.VolumeControl=MapRange.MapRange:Float(Out Range B=^Loop.MainLoopVolMax, Out Range A=^Loop.MainLoopVolMin, In=@n5.Value).Out Value
   )
   AudioMixerAudioMixerStereo2_1 = "AudioMixer.Audio Mixer (Stereo, 2)"(
    In 1 L=@n3.OutL
    In 1 R=@n3.OutR
    In 0 L=@n6.OutL
    In 0 R=@n6.OutR
    Gain 1=MapRange.MapRange:Float(In=VariableAccessor.Float(Variable=@n5.Variable).Value) {also {defaults {In Range A=0.6}}}.Out Value
   )
   n7 = InitVariable.Trigger()
   n8 = &MSP_OneshotGate(Value=AbsoluteValue.Abs:Float(Input=UE.InterpTo:Audio(Target=^Roll).Value).Value) {also {defaults {EnterCondition=3, ExitCondition=4, OneshotThreshold=0.75, RetriggerDelay=0.7}}}
   ArrayRandomGet_0 = Array.Random Get:WaveAsset:Array(
    In Array=^OneShots.Turn
    Next=VariableAccessor.Trigger(Variable=VariableMutator.Trigger(Variable=@n7.Variable, Value=@n8.OnGateEnter).Variable).Value
   )
   WavePlayer_0 = UE.Wave Player:Stereo(Play=@ArrayRandomGet_0.On Next, Wave Asset=@ArrayRandomGet_0.Value) {also {version [1,1]}}
   AudioMixerAudioMixerStereo2_0 = "AudioMixer.Audio Mixer (Stereo, 2)"(
    Gain 0=^Mix.LoopGain
    Gain 1=^Mix.OneShotGain
    In 0 L=@AudioMixerAudioMixerStereo2_1.Out L
    In 0 R=@AudioMixerAudioMixerStereo2_1.Out R
    In 1 L=@WavePlayer_0.Out Left
    In 1 R=@WavePlayer_0.Out Right
   )
   Multiply_0 = UE.Multiply:Audio by Float(AdditionalOperands=^Mix.MasterGain, PrimaryOperand=@AudioMixerAudioMixerStereo2_0.Out R)
   Multiply_1 = UE.Multiply:Audio by Float(AdditionalOperands=^Mix.MasterGain, PrimaryOperand=@AudioMixerAudioMixerStereo2_0.Out L)
   n9 = &MSP_SpatialTravelVerticalDetection(Elevation.Movement=^Movement.SpeedElevation) {also {defaults {VerticalFloor=-60, VerticalRoof=60}}}
   VariableMutator.Trigger(Variable=InitVariable.Trigger().Variable, Value=@n9.Fall)
   VariableMutator.Trigger(Variable=InitVariable.Trigger().Variable, Value=@n9.Rise)
  ]
  links [{dst="@UE.OutputFormat.Stereo.Audio:1", src=Multiply_0.Out}, {dst="@UE.OutputFormat.Stereo.Audio:0", src=Multiply_1.Out}]
 }
]
properties {
 assetImportData=&self:AssetImportData
 attenuationSettings=/Game/Sounds/Creative/Attenuations/Fort_Gadget_MoveTool_Att
 bLooping=true
 duration=10000
 maxDistance=4000
 numChannels=2
 outputFormat=Stereo
 sampleRate=48000
 soundClassObject=&Fort_Gadgets_Items
}
