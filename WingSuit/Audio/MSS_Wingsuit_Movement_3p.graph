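# Looping mono MetaSound source for third-person wingsuit movement, combining a glide loop with one-shot fly-bys.
# The spatial-travel loop fades in on play, follows normalized movement speed for volume, and is summed to mono.
# A distance gate with retrigger delay plus a minimum-speed check fires random fly-bys, scaled by speed using MinDistanceAllowed as the range's lower bound.
# The Mobile page keeps only the loop, omitting fly-bys.
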MetaSoundSource /WingSuit/Audio/MSS_Wingsuit_Movement_3p
inputs [
 FlyBy {type=WaveAsset:Array}
 Loop.MainMovement {type=WaveAsset:Array}
 MinDistanceAllowed {type=Float}
 MinSpeedVal {type=Float}
 Movement.Speed {type=Float}
 Movement.SpeedAzimuth {type=Float}
 Movement.SpeedElevation {type=Float}
 UE.Attenuation.Distance {type=Float}
 UE.Source.OnPlay {type=Trigger}
]
interfaces [Fort.Movement:1.0, UE.Attenuation:1.0, UE.OutputFormat.Mono:1.0, UE.Source:1.1]
outputs ["UE.OutputFormat.Mono.Audio:0" {type=Audio}]
pages [
 Default {
  is_build_graph=true
  input_defaults {
   MinDistanceAllowed=1200
   MinSpeedVal=532.0001
   Movement.Speed=0
   Movement.SpeedAzimuth=0
   Movement.SpeedElevation=0
   UE.Attenuation.Distance=0
   UE.Source.OnPlay=null
   FlyBy [
    /WingSuit/Audio/wav/Wingsuit_FlyBy_01
    /WingSuit/Audio/wav/Wingsuit_FlyBy_02
    /WingSuit/Audio/wav/Wingsuit_FlyBy_03
    /WingSuit/Audio/wav/Wingsuit_FlyBy_04
    /WingSuit/Audio/wav/Wingsuit_FlyBy_05
   ]
   Loop.MainMovement [/WingSuit/Audio/wav/Wingsuit_Flying_Glide_Loop_01_3P]
  }
  nodes [
   n1 = /FortMetasound/SpatialTravel/MSP_SpatialTravelLoop(
    Loop.Sounds=^Loop.MainMovement
    FadeIn=^UE.Source.OnPlay
    Loop.VolumeControl=MapRange.MapRange:Float(In=/FortMetasound/SpatialTravel/MSP_SpatialTravelController(Speed.Movement=^Movement.Speed) {also {defaults {Speed.SpeedMax=2200, Speed.SpeedMin=700}}}.NormalizedSpeed) {also {defaults {Out Range A=0.7}}}.Out Value
   )
   ArrayRandomGet_0 = Array.Random Get:WaveAsset:Array(
    In Array=^FlyBy
    Next=TriggerCompare.Clamp:Float(
     B=^MinSpeedVal
     A=^Movement.Speed
     Compare=&MSP_OneshotGate(OneshotThreshold=^MinDistanceAllowed, Value=^UE.Attenuation.Distance) {also {defaults {EnterCondition=4, ExitCondition=3, RetriggerDelay=3}}}.OnGateEnter
    ) {also {defaults {Type=3}}}.True
   )
   WavePlayer_0 = UE.Wave Player:Mono(Play=@ArrayRandomGet_0.On Next, Wave Asset=@ArrayRandomGet_0.Value) {also {version [1,1]}}
   AudioMixerAudioMixerMono2_0 = "AudioMixer.Audio Mixer (Mono, 2)"(
    In 0=UE.Add:Audio(PrimaryOperand=@n1.OutL, AdditionalOperands=@n1.OutR).Out
    In 1=UE.Multiply:Audio by Float(
     AdditionalOperands=MapRange.MapRange:Float(In Range A=^MinDistanceAllowed, In=^Movement.Speed) {also {defaults {In Range B=2200, Out Range A=0.5}}}.Out Value
     PrimaryOperand=@WavePlayer_0.Out Mono
    ).Out
   ) {also {defaults {Gain 1=2}}}
  ]
  links [{dst="@UE.OutputFormat.Mono.Audio:0", src=AudioMixerAudioMixerMono2_0.Out}]
 }
 Mobile {
  nodes [
   n1 = &MSP_SpatialTravelLoop(
    Loop.Sounds=^Loop.MainMovement
    FadeIn=^UE.Source.OnPlay
    Loop.VolumeControl=MapRange.MapRange:Float(In=&MSP_SpatialTravelController(Speed.Movement=^Movement.Speed) {also {defaults {Speed.SpeedMax=2200, Speed.SpeedMin=700}}}.NormalizedSpeed) {also {defaults {Out Range A=0.7}}}.Out Value
   )
   AudioMixerAudioMixerMono2_0 = "AudioMixer.Audio Mixer (Mono, 2)"(In 0=UE.Add:Audio(PrimaryOperand=@n1.OutL, AdditionalOperands=@n1.OutR).Out) {also {defaults {Gain 1=2}}}
  ]
  links [{dst="@UE.OutputFormat.Mono.Audio:0", src=AudioMixerAudioMixerMono2_0.Out}]
 }
]
properties {
 assetImportData=&self:AssetImportData
 attenuationSettings=&/Att_WingSuit_Movement
 bLooping=true
 duration=10000
 maxDistance=12000
 sampleRate=48000
 soundClassObject=&Fort_Gliders_InAir
}
