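# Sound cue for the wingsuit gadget, named and built to signal that its cooldown has ended.
# A single cooldown-complete wave player feeds a modulator that randomizes volume slightly on each play.
# It uses the shared wingsuit foley attenuation asset and the gadgets/items sound class.
# A concurrency override on the cue caps simultaneous instances and sets a retrigger time.
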SoundCue /WingSuit/Audio/SC_Wingsuit_Cooldown_End
root=Modulator_1
nodes [
 {id=Modulator_1, children [WavePlayer_0], properties {pitchMax=1, volumeMax=1, volumeMin=0.7}}
 {id=WavePlayer_0, properties {soundWaveAssetPtr=/WingSuit/Audio/wav/Wingsuit_Cooldown_Complete_01}}
]
properties {
 attenuationSettings=&/Att_WingSuit_Foley
 bOverrideConcurrency=true
 maxDistance=1100
 soundClassObject=&Fort_Gadgets_Items
 volumeMultiplier=1
 attenuationOverrides {distanceAlgorithm=NaturalSound, falloffDistance=3000, attenuationShapeExtents {x=1000}}
 concurrencyOverrides {maxCount=3, retriggerTime=1}
 concurrencySet [null]
}
