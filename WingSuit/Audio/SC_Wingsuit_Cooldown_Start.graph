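# Sound cue for the wingsuit's cooldown-start sound, classed under gadget and item audio.
# A single wave plays through a modulator that randomizes volume and pitch, then into a mixer whose second input is empty.
# A shared wingsuit foley attenuation asset handles spatialization, and the cue's own concurrency override caps simultaneous instances and sets a retrigger time.
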SoundCue /WingSuit/Audio/SC_Wingsuit_Cooldown_Start
root=Mixer_0
nodes [
 {id=Mixer_0, children [Modulator_1,null], properties {inputVolume [1,0.7]}}
 {id=Modulator_1, children [WavePlayer_4], properties {pitchMax=1, volumeMax=1, volumeMin=0.7}}
 {id=WavePlayer_4, properties {soundWaveAssetPtr=/WingSuit/Audio/wav/Wingsuit_Cooldown_Start_01}}
]
properties {
 attenuationSettings=&/Att_WingSuit_Foley
 bOverrideConcurrency=true
 maxDistance=1100
 soundClassObject=&Fort_Gadgets_Items
 volumeMultiplier=1.5
 attenuationOverrides {distanceAlgorithm=NaturalSound, falloffDistance=3000, attenuationShapeExtents {x=1000}}
 concurrencyOverrides {maxCount=3, retriggerTime=1}
 concurrencySet [null]
}
