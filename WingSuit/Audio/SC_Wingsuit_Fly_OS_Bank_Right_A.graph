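# SoundCue for a one-shot wingsuit wing sound when banking right in flight, classed as an in-air glider sound.
# An equal-weight Random node picks one of four bank-right wave variations and feeds it through a Modulator.
# The cue raises pitch, lowers volume and uses a shared movement-details attenuation asset.
# Its concurrency override caps simultaneous plays and sets a retrigger time.
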SoundCue /WingSuit/Audio/SC_Wingsuit_Fly_OS_Bank_Right_A
root=Modulator_1
nodes [
 {id=Modulator_1, children [Random_0]}
 {id=Random_0, children [WavePlayer_1,WavePlayer_2,WavePlayer_3,WavePlayer_4], properties {weights [1,1,1,1]}}
 {id=WavePlayer_1, properties {soundWaveAssetPtr=/WingSuit/Audio/wav/Wingsuit_Wings_Bank_Right_01}}
 {id=WavePlayer_2, properties {soundWaveAssetPtr=/WingSuit/Audio/wav/Wingsuit_Wings_Bank_Right_02}}
 {id=WavePlayer_3, properties {soundWaveAssetPtr=/WingSuit/Audio/wav/Wingsuit_Wings_Bank_Right_03}}
 {id=WavePlayer_4, properties {soundWaveAssetPtr=/WingSuit/Audio/wav/Wingsuit_Wings_Bank_Right_04}}
]
properties {
 attenuationSettings=&/Att_WingSuit_MovementDetails
 bOverrideConcurrency=true
 maxDistance=8400
 pitchMultiplier=1.3
 soundClassObject=&Fort_Gliders_InAir
 volumeMultiplier=0.3972
 attenuationOverrides {distanceAlgorithm=NaturalSound, falloffDistance=3000, attenuationShapeExtents {x=1000}}
 concurrencyOverrides {maxCount=3, retriggerTime=1}
 concurrencySet [null]
}
