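# Sound cue for the wingsuit in flight, named as a bank-up variant, built from wing-opening recordings.
# A Random node picks one of four equally weighted wave files, then a Modulator varies its playback slightly.
# It plays in the in-air glider sound class with wingsuit movement-detail attenuation.
# Its own concurrency override caps simultaneous instances and sets a retrigger time.
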SoundCue /WingSuit/Audio/SC_Wingsuit_Fly_OS_Bank_Up_B
root=Modulator_1
nodes [
 {id=Modulator_1, children [Random_1], properties {pitchMax=1, volumeMax=1.1, volumeMin=1}}
 {id=Random_1, children [WavePlayer_5,WavePlayer_6,WavePlayer_7,WavePlayer_8], properties {weights [1,1,1,1]}}
 {id=WavePlayer_5, properties {soundWaveAssetPtr=/WingSuit/Audio/wav/Wingsuit_Wings_Open_01}}
 {id=WavePlayer_6, properties {soundWaveAssetPtr=/WingSuit/Audio/wav/Wingsuit_Wings_Open_02}}
 {id=WavePlayer_7, properties {soundWaveAssetPtr=/WingSuit/Audio/wav/Wingsuit_Wings_Open_03}}
 {id=WavePlayer_8, properties {soundWaveAssetPtr=/WingSuit/Audio/wav/Wingsuit_Wings_Open_04}}
]
properties {
 attenuationSettings=&/Att_WingSuit_MovementDetails
 bOverrideConcurrency=true
 maxDistance=8400
 soundClassObject=&Fort_Gliders_InAir
 volumeMultiplier=0.7943
 attenuationOverrides {distanceAlgorithm=NaturalSound, falloffDistance=3000, attenuationShapeExtents {x=1000}}
 concurrencyOverrides {maxCount=3, retriggerTime=1}
 concurrencySet [null]
}
