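# Sound cue for the wingsuit's barrel roll wing sound, a one-shot played during flight in the gliders in-air sound class.
# A Random node picks among four equally weighted wave variations, and a Modulator at the root adds slight pitch and volume variation.
# It uses the wingsuit movement-detail attenuation asset and overrides concurrency to cap simultaneous instances and limit retriggering.
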SoundCue /WingSuit/Audio/SC_Wingsuit_Fly_OS_Barrel_Roll
root=Modulator_1
nodes [
 {id=Modulator_1, children [Random_0], properties {pitchMax=1, volumeMax=1.1, volumeMin=1}}
 {id=Random_0, children [WavePlayer_0,WavePlayer_1,WavePlayer_2,WavePlayer_3], properties {weights [1,1,1,1]}}
 {id=WavePlayer_0, properties {soundWaveAssetPtr=/WingSuit/Audio/wav/Wingsuit_Wings_BarrelRoll_01}}
 {id=WavePlayer_1, properties {soundWaveAssetPtr=/WingSuit/Audio/wav/Wingsuit_Wings_BarrelRoll_02}}
 {id=WavePlayer_2, properties {soundWaveAssetPtr=/WingSuit/Audio/wav/Wingsuit_Wings_BarrelRoll_03}}
 {id=WavePlayer_3, properties {soundWaveAssetPtr=/WingSuit/Audio/wav/Wingsuit_Wings_BarrelRoll_04}}
]
properties {
 attenuationSettings=&/Att_WingSuit_MovementDetails
 bOverrideConcurrency=true
 maxDistance=8400
 soundClassObject=&Fort_Gliders_InAir
 volumeMultiplier=0.7943
 attenuationOverrides {distanceAlgorithm=NaturalSound, falloffDistance=3000, attenuationShapeExtents {x=1000}}
 concurrencyOverrides {maxCount=3, retriggerTime=1}
 concurrencySet [null]
}
