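# Sound cue for wingsuit landing foley, named for the land event and built from cloth-shake recordings.
# A Random node picks among fourteen equally weighted Wingsuit_Foley_Shake waves, and a Modulator varies pitch and volume on each play.
# It routes to the player foley sound class, uses the shared wingsuit foley attenuation asset, and overrides concurrency to cap simultaneous instances.
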SoundCue /WingSuit/Audio/SC_Wingsuit_Foley_Land
root=Modulator_1
nodes [
 {id=Modulator_1, children [Random_0], properties {pitchMax=1, volumeMax=1, volumeMin=0.7}}
 {
  id=Random_0
  children [
   WavePlayer_0
   WavePlayer_1
   WavePlayer_2
   WavePlayer_3
   WavePlayer_7
   WavePlayer_8
   WavePlayer_9
   WavePlayer_10
   WavePlayer_11
   WavePlayer_12
   WavePlayer_13
   WavePlayer_14
   WavePlayer_15
   WavePlayer_16
  ]
  properties {weights [1,1,1,1,1,1,1,1,1,1,1,1,1,1]}
 }
 {id=WavePlayer_0, properties {soundWaveAssetPtr=/WingSuit/Audio/wav/Wingsuit_Foley_Shake_14}}
 {id=WavePlayer_1, properties {soundWaveAssetPtr=/WingSuit/Audio/wav/Wingsuit_Foley_Shake_13}}
 {id=WavePlayer_10, properties {soundWaveAssetPtr=/WingSuit/Audio/wav/Wingsuit_Foley_Shake_07}}
 {id=WavePlayer_11, properties {soundWaveAssetPtr=/WingSuit/Audio/wav/Wingsuit_Foley_Shake_06}}
 {id=WavePlayer_12, properties {soundWaveAssetPtr=/WingSuit/Audio/wav/Wingsuit_Foley_Shake_05}}
 {id=WavePlayer_13, properties {soundWaveAssetPtr=/WingSuit/Audio/wav/Wingsuit_Foley_Shake_04}}
 {id=WavePlayer_14, properties {soundWaveAssetPtr=/WingSuit/Audio/wav/Wingsuit_Foley_Shake_03}}
 {id=WavePlayer_15, properties {soundWaveAssetPtr=/WingSuit/Audio/wav/Wingsuit_Foley_Shake_02}}
 {id=WavePlayer_16, properties {soundWaveAssetPtr=/WingSuit/Audio/wav/Wingsuit_Foley_Shake_01}}
 {id=WavePlayer_2, properties {soundWaveAssetPtr=/WingSuit/Audio/wav/Wingsuit_Foley_Shake_12}}
 {id=WavePlayer_3, properties {soundWaveAssetPtr=/WingSuit/Audio/wav/Wingsuit_Foley_Shake_11}}
 {id=WavePlayer_7, properties {soundWaveAssetPtr=/WingSuit/Audio/wav/Wingsuit_Foley_Shake_10}}
 {id=WavePlayer_8, properties {soundWaveAssetPtr=/WingSuit/Audio/wav/Wingsuit_Foley_Shake_09}}
 {id=WavePlayer_9, properties {soundWaveAssetPtr=/WingSuit/Audio/wav/Wingsuit_Foley_Shake_08}}
]
properties {
 attenuationSettings=&/Att_WingSuit_Foley
 bOverrideConcurrency=true
 maxDistance=1100
 soundClassObject=&Fort_Player_Foley
 volumeMultiplier=1
 attenuationOverrides {distanceAlgorithm=NaturalSound, falloffDistance=3000, attenuationShapeExtents {x=1000}}
 concurrencyOverrides {maxCount=3, retriggerTime=1}
 concurrencySet [null]
}
