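# Sound cue for the wingsuit rocket launch, assigned to the in-air glider sound class.
# A single launch wave plays through a modulator that slightly randomizes pitch and volume.
# The root switch reads the PlayerAffiliation parameter, but all four branches share that modulator, so every affiliation sounds the same.
# It references shared wingsuit movement attenuation and overrides concurrency to limit simultaneous instances and retriggering.
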SoundCue /WingSuit/Audio/SC_Wingsuit_Rocket_Launch
root=Switch_0
nodes [
 {id=Modulator_0, children [WavePlayer_0], properties {pitchMax=1, volumeMax=1.1, volumeMin=1}}
 {id=Switch_0, children [Modulator_0,Modulator_0,Modulator_0,Modulator_0], properties {intParameterName=PlayerAffiliation}}
 {id=WavePlayer_0, properties {soundWaveAssetPtr=/WingSuit/Audio/wav/Wingsuit_Rocket_Launch_01}}
]
properties {
 attenuationSettings=&/Att_WingSuit_Movement
 bOverrideConcurrency=true
 maxDistance=12000
 soundClassObject=&Fort_Gliders_InAir
 volumeMultiplier=1
 attenuationOverrides {distanceAlgorithm=NaturalSound, falloffDistance=3000, attenuationShapeExtents {x=1000}}
 concurrencyOverrides {maxCount=3, retriggerTime=1}
 concurrencySet [null]
}
