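# Alpha-composite, Default Lit material for Niagara ribbons in the WingSuit VFX, built around a diagonal caustic-wind texture.
# UV tilt, swizzle and a vector-driven panner, timed by a dynamic particle parameter, shape texture scrolling.
# Named reroutes organize the texture, UV-gradient edge mask, fresnel, depth fade and camera fade terms.
# Particle color tints emissive and scales opacity, while camera fade statically switches between view-axis distance and pixel depth.
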Material /WingSuit/VFX/Material/M_Ribbon
graph {
 properties {bAutomaticallySetUsageInEditor=false, bUsedWithNiagaraRibbons=true, blendMode=BLEND_AlphaComposite, usedShadingModels=Default Lit}
 nodes [
  n1 = NamedRerouteUsage()
  n2 = Multiply(
   A=TextureSample(UVs=NamedRerouteUsage()) {samplerType=SAMPLERTYPE_LinearGrayscale, texture=/Game/Effects/Fort_Effects/Textures/Tile/T_Caustic_Wind_Diagonal}.G
   B=@n1
  )
  LinearInterpolate(A=Multiply(A=@n1) {constB=0.25}, B=@n2, Alpha=ComponentMask(None=TextureCoordinate() {coordinateIndex=1}) {r=true})
  n3 = TextureCoordinate()
  LinearInterpolate(
   A=@n3
   B=MaterialFunctionCall(XY=@n3) {materialFunction=&Swizzle}.YX
   Alpha=ScalarParameter() {defaultValue=1, parameterName=Main_UV_Swizzle}
  )
  MaterialFunctionCall(Emissive=NamedRerouteUsage()) {materialFunction=&MF_FX_EmissiveAdjustments}
  n4 = TextureCoordinate()
  n5 = TextureCoordinate() {coordinateIndex=1}
  n6 = ScalarParameter() {parameterName=CubeMap_Switch}
  n7 = Subtract(A=Multiply(A=LinearInterpolate(A=@n4, B=@n5, Alpha=@n6)) {constB=2})
  Multiply(A=@n7, B=@n7) {constB=2}
  n8 = NamedRerouteUsage()
  n9 = NamedRerouteUsage()
  Multiply(A=Multiply(A=Multiply(A=@n8, B=@n9), B=NamedRerouteUsage()), B=NamedRerouteUsage())
  n10 = Multiply(A=@n2, B=NamedRerouteUsage())
  n11 = NamedRerouteUsage()
  NamedRerouteDeclaration(None=Multiply(A=Multiply(A=@n10, B=@n11), B=NamedRerouteUsage())) {name=Texture, nodeColor=<rgba 0 1 0.5078 1>}
  n12 = VectorParameter() {parameterName=Main_UV, defaultValue=<rgba 0.5 0.25 -0.2657 0.208>}
  NamedRerouteDeclaration(None=MaterialFunctionCall!(
    ClampTexture
    UV=NamedRerouteUsage()
    Tex
    InRGB=@n12.RGB
    InA=@n12.RGB
    Time=DynamicParameter() {defaultValue=<rgba 1 1 1 250>, paramNames [Time,Distort,Dissolve,WPO_Width]}.Time
    TimeMult
   ) {materialFunction=/Game/Effects/Niagara/MasterMaterials/Functions/MF_FX_VectorDrive_TexPanner}.PannerUV) {name=MainUV, nodeColor=<rgba 0 0.4049 1 1>}
  n13 = Subtract(A=CameraPositionWS(), B=WorldPosition().XYZ)
  n14 = Length(None=@n13) {is=class MaterialExpressionLength}
  n15 = DotProduct(A=Normalize(VectorInput=@n13), B=Transform(None=Constant3Vector() {constant=<rgba 0 0 -1 0>}) {transformSourceType=TRANSFORMSOURCE_View})
  n16 = ScalarParameter() {defaultValue=20, parameterName=Camfade_Offset}
  n17 = ScalarParameter() {defaultValue=40, parameterName=Camfade_Lemgth}
  n18 = Divide(A=Subtract(A=Multiply(A=@n14, B=@n15), B=@n16), B=@n17)
  n19 = Divide(A=Subtract(A=PixelDepth().R, B=@n16), B=@n17)
  n20 = StaticBool() {value=true}
  NamedRerouteDeclaration(None=Saturate(None=StaticSwitch(True=@n18, False=@n19, Value=@n20))) {name=CamFade, nodeColor=<rgba 1 0 0.9613 1>}
  n21 = CameraPositionWS()
  n22 = WorldPosition()
  n23 = DotProduct(A=VertexNormalWS(), B=Normalize(VectorInput=Subtract(A=@n21, B=@n22.XYZ)))
  n24 = Abs(None=@n23)
  n25 = Power(
   Base=LinearInterpolate(A=OneMinus(None=@n24), B=@n24, Alpha=ScalarParameter() {defaultValue=1, parameterName="Fresnel_Invert (0or1)"})
   Exp=ScalarParameter() {defaultValue=0.5, parameterName=Fresnel_Power}
  )
  n26 = ScalarParameter() {defaultValue=0.05, parameterName=Fresnel_Offset}
  NamedRerouteDeclaration(None=Multiply(A=Subtract(A=@n25, B=@n26), B=Add(A=@n26))) {name=Fresnel, nodeColor=<rgba 1 0 0.9613 1>}
  NamedRerouteDeclaration(None=DepthFade(FadeDistance=ScalarParameter() {defaultValue=50, group=Opacity, parameterName=DepthFade})) {name=DepthFade, nodeColor=<rgba 0.8842 0 1 1>}
  NamedRerouteDeclaration(None=Multiply(A=ParticleColor().RGB, B=NamedRerouteUsage())) {name=Emissive, nodeColor=<rgba 0 0 0 1>}
  NamedRerouteDeclaration(None=Multiply(A=ParticleColor().A, B=@n8)) {name=Opacity, nodeColor=<rgba 0 0 0 1>}
  n27 = MaterialFunctionCall(Float2=TextureCoordinate()) {materialFunction=&BreakOutFloat2Components}
  n28 = Multiply(A=@n27.R, B=ScalarParameter() {defaultValue=0.4898, parameterName=UV_Tilt})
  NamedRerouteDeclaration(None=MaterialFunctionCall(X=@n27.R, Y=Add(A=@n27.G, B=@n28)) {materialFunction=&MakeFloat2}.Result) {name=TiltedUV, nodeColor=<rgba 0 0.3848 1 1>}
  n29 = MaterialFunctionCall(Float2=OneMinus(None=Abs(None=@n7))) {materialFunction=&BreakOutFloat2Components}
  NamedRerouteDeclaration(None=MaterialFunctionCall(A=@n29.R, B=Multiply(A=@n29.R, B=@n29.R), C=@n29.R, Alpha=ScalarParameter() {defaultValue=1, parameterName=CubeMap_Select}) {materialFunction=&Lerp_3Color}.Result) {name=CubeMap_2, nodeColor=<rgba 1 1 1 1>}
  NamedRerouteUsage()
  Reroute(None=@n23)
 ]
}
