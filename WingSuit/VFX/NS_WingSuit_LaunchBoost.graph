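# Niagara launch-boost effect for the wingsuit: a burst of orange HDR mesh cone with an accompanying light flash.
# Cone renders through two mesh renderers, shrinking rapidly from a stretched X-axis scale while its alpha fades in and out and a dynamic material parameter ramps over life.
# Light spawns a locally offset orange point light whose brightness decays to zero.
# Declared user variables are not referenced by the shown modules.
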NiagaraSystem /WingSuit/VFX/NS_WingSuit_LaunchBoost
dependencies {
 usedDataInterfaces [&NiagaraDataInterfaceVectorCurve, &NiagaraDataInterfaceCurve]
 usedDynamicInputs []
 usedModules [&SystemState]
}
emitters {
 Cone {
  emitterUpdateScript [
   &EmitterState(Life Cycle Mode=::NewEnumerator1, Loop Behavior=::NewEnumerator1, Loop Duration=3, UseLoopDelay=false)
   &SpawnBurst_Instantaneous(Use Spawn Probability=false, UseLoopCountLimit=false)
  ]
  particleSpawnScript [
   &InitializeParticle(
    Color=<rgba 50 10.4544 1.8229 0.8696>
    Lifetime=&RandomRangeFloat(Minimum=1, Maximum=1)
    Mesh Scale Mode=::NewEnumerator3
    Mesh Uniform Scale=2
    UsePositionOffset=false
    Write Material Random=true
   )
   &InitialMeshOrientation(
    Mesh Orientation Mode=::NewEnumerator1
    Rotation=&RandomRangeVector(Minimum=<0 0 0>, Maximum=<0 0 0>)
    Use Rotation Vector=true
   )
  ]
  particleUpdateScript [
   &ParticleState()
   &ScaleMeshSize(Scale Factor=&VectorFromCurve(VectorCurve=DataInterface() {propertyValues {
       XCurve {keys [
         {arriveTangent=-28.6719, interpMode=RCIM_Cubic, leaveTangent=-28.6719, tangentMode=RCTM_User, time=0, value=30}
         {arriveTangent=-47.0205, interpMode=RCIM_Cubic, leaveTangent=-47.0205, time=0.0938, value=10.7609}
         <at 0.5684 3.274 RCIM_Cubic>
        ]}
       YCurve {keys [
         <at 0.0156 8.9334 RCIM_Cubic>
         {arriveTangent=-2.3898, interpMode=RCIM_Cubic, leaveTangent=-2.3898, tangentMode=RCTM_User, time=0.208, value=1.6657}
         <at 1 1>
        ]}
       ZCurve {keys [
         <at 0.0156 8.9334 RCIM_Cubic>
         {arriveTangent=-2.3898, interpMode=RCIM_Cubic, leaveTangent=-2.3898, tangentMode=RCTM_User, time=0.208, value=1.6657}
         <at 1 1>
        ]}
      }}))
   &ScaleColor(
    Scale Alpha=&FloatFromCurve(
     FloatCurve=DataInterface() {propertyValues {Curve {
        defaultValue=1
        keys [
         {interpMode=RCIM_Cubic, tangentMode=RCTM_User, time=0, value=0}
         {arriveTangent=-3.2814, interpMode=RCIM_Cubic, leaveTangent=-3.2814, tangentMode=RCTM_User, time=0.0727, value=1}
         {arriveTangent=-1.0784, interpMode=RCIM_Cubic, leaveTangent=-1.0784, time=0.3273, value=0.8265}
         {tangentMode=RCTM_User, time=1, value=0}
        ]
       }}}
     Scale Curve=0.1
    )
    ScaleA=true
    ScaleRGB=true
   )
   /SetVariables(Particles.Color001:LinearColor=&MultiplyLinearColor(Linear Color=$Particles.Color, Scale Factor=<10 10 10 1>)) {isSetParameters=true}
   &DynamicMaterialParameters(
    Index 0 Param 1=&FloatFromCurve(
     FloatCurve=DataInterface() {propertyValues {Curve {defaultValue=1, keys [{tangentMode=RCTM_User, time=0, value=0}, {tangentMode=RCTM_User, time=1, value=1}]}}}
     Scale Curve=10
    )
    Index 0 Param 4=1
    Module.Index0Float0Write=true
    Module.Index0Float1Write=true
    Module.Index0Float2Write=true
    Module.Index0Float3Write=true
   )
  ]
  rendererClasses [&NiagaraMeshRendererProperties]
  renderers [
   {rendererClass=&NiagaraMeshRendererProperties, rendererIndex=0}
   {rendererClass=&NiagaraMeshRendererProperties, rendererIndex=1}
  ]
 }
 Light {
  emitterUpdateScript [
   &EmitterState()
   &SpawnBurst_Instantaneous(Use Spawn Probability=false, UseLoopCountLimit=false)
  ]
  particleSpawnScript [
   &InitializeParticle(
    Color=<rgba 20 6.3043 1.7391 1>
    Position Offset=<30 0 2>
    Position Offset Coordinate Space=::Local
    UsePositionOffset=true
    Write Material Random=true
   )
   /SetVariables(Particles.LightExponent=4, Particles.LightRadius=500) {isSetParameters=true}
  ]
  particleUpdateScript [
   &ParticleState()
   &ScaleColor(
    Scale RGB=&VectorFromFloat(Value=&FloatFromCurve(FloatCurve=DataInterface() {propertyValues {Curve {keys [
          {arriveTangent=-31.0453, interpMode=RCIM_Cubic, leaveTangent=-31.0453, tangentMode=RCTM_User, time=0, value=10}
          {arriveTangent=-8.9566, interpMode=RCIM_Cubic, leaveTangent=-8.9566, time=0.2132, value=2.5657}
          {arriveTangent=-3.2609, interpMode=RCIM_Cubic, leaveTangent=-3.2609, time=0.8572, value=2.3226}
          <at 1 0 RCIM_Cubic>
         ]}}}))
    ScaleA=true
    ScaleRGB=true
   )
  ]
  rendererClasses [&NiagaraLightRendererProperties]
  renderers [{rendererClass=&NiagaraLightRendererProperties, rendererIndex=0}]
 }
}
summary {
 systemName=NS_WingSuit_LaunchBoost
 emitters [
  {emitterName=Cone, rendererClasses [&NiagaraMeshRendererProperties]}
  {emitterName=Light, rendererClasses [&NiagaraLightRendererProperties]}
 ]
 userVariables [
  User.Ascended {type=NiagaraBool, defaultValue=true}
  User.bIsFrontEnd {type=NiagaraBool}
  User.bIsFrontEndPreview {type=NiagaraBool}
  User.Finish Signal Duration {type=NiagaraFloat, defaultValue=1.15, description=""}
  User.Is Air {type=NiagaraFloat, defaultValue=1, description=""}
 ]
}
